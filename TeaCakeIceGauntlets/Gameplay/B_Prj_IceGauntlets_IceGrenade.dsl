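; Ice grenade projectile for the Ice Gauntlets weapon that explodes on a blocking touch or when its tunable lifetime expires.
; BeginPlay starts the lifetime timer and offsets the Niagara trail's muzzle position by the dual weapon's next firing hand.
; Touches clear the timer and have the server add EQS-based icy patches through a GameState manager component, if enabled; timeouts spawn none.
; OnStop ends in-air audio and VFX, covering clients that miss OnTouched.
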(event EventBeginPlay
  ;[0.02391 0.15 0.02888 0.5] Begin Play
  (bind _returnvalue (Utilities|Time|SetTimerbyEvent (EventDispatchers|CreateEvent :!event Custom|LifetimeEnded) (ScalableFloat|GetValueatLevel (Variables|Hotfix|GetRow_LifetimeBeforeExploding))))
  (Variables|Default|SetLifetimeTimer _returnvalue)
  (bind _returnvalue_2 (CallFunction|GetOwnerWeapon))
  (Utilities|IsValid _returnvalue_2
    (:Is%20Valid
      (bind _asb_weap_ice_gauntlets (Utilities|Casting|CastToB_Weap_IceGauntlets _returnvalue_2))
      (Effects|Components|ParticleSystem|SetVectorParameter (Variables|Default|GetNS_IceGauntlets_BigBomb_Projectile) "Muzzle Pos" (+ (Transformation|GetActorLocation _returnvalue_2) (Utilities|Select :Index (RangedWeaponDual|GetNextFireHand _asb_weap_ice_gauntlets) :LEFT (* (Transformation|GetRightVector (Variables|/Script/Engine.Actor|GetRootComponent _returnvalue_2)) "51.000000,53.000000,0.000000") :RIGHT "0, 0, 0"))))))

(event Custom|LifetimeEnded
  ;[0.15 0.04442 0.02717 0.5] Lifetime Ended
  (Game|Components|ProjectileMovement|StopSimulating (Variables|Movement|GetProjectileMovementComponent) (Collision|MakeHitResult false false 0.0 0.0 (Transformation|GetActorLocation) "0, 0, 0" (Math|Vector|Normalize (Transformation|GetVelocity))))
  (Networking|SwitchHasAuthority
    (:Authority
      (Delivery|DoExplosionAsync))
    (:Remote
      (Networking|CanExecuteCosmeticEvents
        (:True
          (Components|Activation|Deactivate (Variables|Default|GetNS_IceGauntlets_BigBomb_Projectile)))))))

(event EventOnTouched
  (params (OtherActor :type Actor%20Object%20Reference) (OtherComp :type Primitive%20Component%20Object%20Reference) (HitResult :type Hit%20Result%20Structure%20%28by%20ref%29) (bIsOverlap :type Boolean))
  ;[0.04325 0.15 0.09947 0.5] On Touched
  ; Used instead of OnHit because it is called on client by parent class if hit occurs before client projectile spawned
  (|CallParentFunction OtherActor OtherComp HitResult bIsOverlap)
  (if (not bIsOverlap)
    (Delivery|DoExplosionAsync)
    (Utilities|Time|ClearandInvalidateTimerbyHandle (Variables|Default|GetLifetimeTimer))
    (Networking|SwitchHasAuthority
      (:Authority
        (CallFunction|SpawnIcyPatches)))))

(event EventOnStop
  (params (Hit :type Hit%20Result%20Structure%20%28by%20ref%29))
  ;[0.15 0.06903 0.02658 0.5] On Stop
  ; Used to ensure any time projectile is stopped, in air FX stop playing. Sometimes OnTouched is not called on client
  (|CallParentFunction Hit)
  (Networking|CanExecuteCosmeticEvents
    (:True
      (Audio|Components|Audio|FadeOut (Variables|Default|GetAudio_InAirLoop) 0.1 0.0 "Linear" :!class /Script/Engine.AudioComponent)
      (Components|Activation|Deactivate (Variables|Default|GetNS_IceGauntlets_BigBomb_Projectile)))))

(fn SpawnIcyPatches
  (bind _returnvalue (Fortnite|Athena|GameState|GetGameStateAthena))
  (bind _returnvalue_2 (Actor|GetComponentbyClass _returnvalue "/TeaCakeIcyShins/Gameplay/B_FortGameStateComponent_IcyPatchManager.B_FortGameStateComponent_IcyPatchManager_C"))
  ;[0.03714 0.1414 0.15 0.5] Spawn icy patches
  (Networking|SwitchHasAuthority
    (:Authority
      (if (ScalableFloat|GetValueasBool (Variables|Hotfix|GetRow_SpawnIcyPatches))
        (Utilities|IsValid _returnvalue_2
          (:Is%20Valid
            (Default|AddIcyPatchatLocation _returnvalue_2 (Transformation|GetActorLocation) (Variables|Default|GetPatch%20EQS%20Query%20Template) (Variables|Default|GetPatch%20Combination%20Radius))))))))

(fn ConstructionScript
  (|CallParentFunction)
  ; VFX
  (Niagara|SetNiagaraVariable(Float) (Variables|Default|GetNS_IceGauntlets_BigBomb_Projectile) "PjrLife" (ScalableFloat|GetValueatLevel (Variables|Hotfix|GetRow_LifetimeBeforeExploding))))
