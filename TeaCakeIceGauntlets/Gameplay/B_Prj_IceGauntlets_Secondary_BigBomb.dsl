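; Big bomb projectile fired by the Ice Gauntlets' secondary attack, able to spawn an ice storm and knock back explosion victims.
; On BeginPlay it schedules delayed gravity and feeds the owning weapon's location to its Niagara effect as muzzle position.
; The server handles the first blocking touch with a delayed storm spawn, and explosions with a batched, async knockback GameplayEffect.
; Hotfix rows drive toggles and tuning; clients disable capsule overlap events.
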(event EventBeginPlay
  (|CallParentFunction)
  (Utilities|Time|SetTimerbyEvent (delegate Custom|EnableGravity) (ScalableFloat|GetValueatLevel (Variables|Hotfix|GetRow_TimeBeforeGravityStarts)))
  (bind _returnvalue (CallFunction|GetOwnerWeapon))
  (Utilities|IsValid _returnvalue
    (:Is%20Valid
      (bind _asb_weap_ice_gauntlets (Utilities|Casting|CastToB_Weap_IceGauntlets _returnvalue))
      (Effects|Components|ParticleSystem|SetVectorParameter (Variables|Default|GetNS_IceGauntlets_BigBomb_Projectile) "Muzzle Pos" (Transformation|GetActorLocation _asb_weap_ice_gauntlets)))))

(event Custom|EnableGravity
  (Variables|/Script/Engine.ProjectileMovementComponent|SetProjectileGravityScale (ScalableFloat|GetValueatLevel (Variables|Hotfix|GetRow_GravityScale)) (Variables|Movement|GetProjectileMovementComponent)))

(event EventOnTouched
  (params (OtherActor :type Actor%20Object%20Reference) (OtherComp :type Primitive%20Component%20Object%20Reference) (HitResult :type Hit%20Result%20Structure%20%28by%20ref%29) (bIsOverlap :type Boolean))
  (|CallParentFunction OtherActor OtherComp HitResult bIsOverlap)
  (if (not bIsOverlap)
    (Networking|SwitchHasAuthority
      (:Authority
        (Utilities|FlowControl|DoOnce false
          (:Completed
            ;[0.09441 0.9525 0.9948 0.5] Spawn Ice Storm
            (if (ScalableFloat|GetValueasBool (Variables|Hotfix|GetRow_SpawnIceStorm))
              (Utilities|FlowControl|Delay (ScalableFloat|GetValueatLevel (Variables|Hotfix|GetRow_IceStormSpawnDelay)))
              (Building|SpawnBuildingActor (Variables|Default|GetIceStormActor_Class) (Transformation|GetActorTransform) (Variables|Actor|GetInstigator) (Variables|Actor|GetInstigator)))))))))

(event EventOnExploded
  (params (HitActors :type Array%20of%20Actor%20Object%20References) (HitResults :type Array%20of%20Hit%20Result%20Structures))
  (|CallParentFunction HitActors HitResults)
  (Networking|SwitchHasAuthority
    (:Authority
      ; Apply Knockback
      (if (ScalableFloat|GetValueasBool (Variables|Hotfix|GetRow_ApplyKnockback))
        (bind _instigator (Variables|Actor|GetInstigator)
          (:then
            (Ability|Tasks|ApplyGameplayEffectToHitResultsAsync (Ability|GetAbilitySystemComponent _instigator) HitResults (Variables|Default|GetGE_ExplosionKnockback) 0 1 (Class|AthenaFunctionLibrary|MakeSimpleFortEffectContext _instigator self) (Variables|Hotfix|GetRow_ExplosionTargetsPerTick)))
          (:else))))))

(fn ConstructionScript
  ;[1 1 1] Turn off the tick and set Generate Overlaps to false on the client.
  (if (not (Networking|IsServer))
    (Variables|/Script/Engine.PrimitiveComponent|SetbGenerateOverlapEvents false (Variables|Collision|GetCapsuleComponent))))
