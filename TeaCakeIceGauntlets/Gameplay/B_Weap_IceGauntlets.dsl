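; Dual-handed Ice Gauntlets ranged weapon that removes itself from the player's inventory and plays a break sound when its ammo runs out.
; BeginPlay binds ammo-consumed and hand-fired events, the latter toggling left/right firing flags on the player's linked gauntlet anim layer.
; A gameplay-tag wait on the owner changes HUD key action visibility when HideSecondary is added, with debug prints left in.
; The weapon disables its sound indicator.
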(event EventBeginPlay
  (Variables|Default|SetPlayerPawn (Game|GetInstigator))
  (EventDispatchers|BindEventtoOnConsumeAmmo (EventDispatchers|CreateEvent :!event Custom|Event_OnConsumeAmmo))
  (RangedWeaponDual|BindEventtoOnHandFired (EventDispatchers|CreateEvent :!event Custom|OnHandFiredEvent))
  (Ability|Tasks|WaitGameplayTagAddToActor (Actor|GetOwner) "(TagName=\"HUD.Athena.SpecialEvent.HideSecondary\")"
    (:then
      (Development|PrintString "NOT IN EVENT SO NOT HIDING" true true "(R=0.000000,G=0.660000,B=1.000000,A=1.000000)" 1e+01))
    (:Added
      (WeaponHUDData|SetHudKeyActionVisibility 1)
      (Development|PrintString "IN EVENT SO HIDING" true true "(R=0.000000,G=0.660000,B=1.000000,A=1.000000)" 1e+01))))

(event EventEndPlay
  (params (EndPlayReason :type EEndPlayReason%20Enum :default "Destroyed"))
  (EventDispatchers|UnbindEventfromOnConsumeAmmo (EventDispatchers|CreateEvent :!event Custom|Event_OnConsumeAmmo)))

(event Custom|Event_OnConsumeAmmo
  (params (Quantity :type Integer))
  (if (<= (FortWeapon|GetRemainingAmmo) 0)
    (Inventory|RemoveItemFromPlayerByGuid (Game|GetInstigatorController) (FortWeapon|GetInventoryGUID) -1 true)
    (CallFunction|PlaySoundonBreak)))

(event Custom|OnHandFiredEvent
  (params (Hand :type EDualWeaponHand%20Enum :default "LEFT") (bPersistantFire :type Boolean))
  (bind _astea_cake_ice_gauntlets_player_anim_bp (Utilities|Casting|CastToTeaCake_IceGauntlets_Player_AnimBP (Components|SkeletalMesh|AnimationBlueprintLinking|GetLinkedAnimLayerInstancebyClass (Variables|/Script/Engine.Character|GetMesh (Utilities|Casting|CastToFortPlayerPawn (Game|GetInstigator))) "/TeaCakeIceGauntlets/Assets/ANIM/TeaCake_IceGauntlets_Player_AnimBP.TeaCake_IceGauntlets_Player_AnimBP_C")))
  (switch Utilities|FlowControl|Switch|SwitchonEDualWeaponHand Hand
    (:LEFT
      (Variables|/TeaCakeIceGauntlets/Assets/ANIM/TeaCake_IceGauntlets_Player_AnimBP.TeaCake_IceGauntlets_Player_AnimBP_C|SetIsFiringRight false _astea_cake_ice_gauntlets_player_anim_bp)
      (Variables|/TeaCakeIceGauntlets/Assets/ANIM/TeaCake_IceGauntlets_Player_AnimBP.TeaCake_IceGauntlets_Player_AnimBP_C|SetIsFiringLeft true _astea_cake_ice_gauntlets_player_anim_bp))
    (:RIGHT
      (Variables|/TeaCakeIceGauntlets/Assets/ANIM/TeaCake_IceGauntlets_Player_AnimBP.TeaCake_IceGauntlets_Player_AnimBP_C|SetIsFiringLeft false _astea_cake_ice_gauntlets_player_anim_bp)
      (Variables|/TeaCakeIceGauntlets/Assets/ANIM/TeaCake_IceGauntlets_Player_AnimBP.TeaCake_IceGauntlets_Player_AnimBP_C|SetIsFiringRight true _astea_cake_ice_gauntlets_player_anim_bp))))

(event Custom|PlaySoundOnBreak
  (Audio|PlaySound2D "/TeaCakeIceGauntlets/Assets/Audio/SC_TeaCake_IceGauntlets_Break.SC_TeaCake_IceGauntlets_Break"))

(fn ShouldShowSoundIndicator
  (params (PlayerController :type Fort%20Player%20Controller%20Object%20Reference))
  (return false))

(fn ConstructionScript)
