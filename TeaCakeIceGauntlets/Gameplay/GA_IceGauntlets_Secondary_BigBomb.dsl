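; Gameplay ability for the Ice Gauntlets secondary "Big Bomb": leaps on activation, then launches an exploding projectile.
; An ItemThrown montage notify (flagged to fix for the final animation) commits the ability and spawns the projectile on authority.
; Launch aim depends on controller type and third-person tag, plus a scalable pitch offset.
; AI-controlled launches add guided motion and broadcast a speech event; failed activation plays the cooldown sound.
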(event Ability|EventActivateAbility
  (|CallParentFunction)
  (CallFunction|Leap))

(event EventCastMontageNotifyStart
  (params (NotifyTag :type Gameplay%20Tag%20Structure))
  ; @Henry fix this when the secondary animation comes in
  (if (== NotifyTag "(TagName=\"Event.Athena.Item.ItemThrown\")")
    (Utilities|PowerCommitAbility (Variables|AbilityCommit|GetIgnoreAbilityCommit)
      (:Pass
        (bind _playerpawn (Variables|Default|GetPlayerPawn)
          (:then
            (Networking|ForceNetUpdate _playerpawn)
            (Ability|HasAuthority
              (:True
                (bind (_location _rotation) (CallFunction|GetProjectileLaunchLocationandRotation _playerpawn))
                (Projectile|SpawnProjectile :ProjectileClass (Variables|Default|GetProjectile%20Class) :RequestedBy (Weapon|GetCurrentWeapon _playerpawn) :SpawnLocation _location :SpawnRotation "0, 0, 0" :SpawnDirection (Math|Rotator|MakeRotator (.roll _rotation) (+ (.pitch _rotation) (ScalableFloat|GetValueatLevel (Variables|Default|GetRow_ProjectilePitchOffset))) (.yaw _rotation)) :EffectContainerSpecToApplyOnExplode (Ability|MakeGameplayEffectContainerSpec "(TagName=\"Gameplay.Application.Explode\")") :bUseDefaultPhysics true :InitialSpeed 0.0 :GravityScale 0.0 :ChargePercent 0.0))))
          (:else))))))

(fn GetProjectileLaunchLocationAndRotation
  (params (Actor :type Fort%20Player%20Pawn%20Object%20Reference))
  (bind _output_get (LocalVariables|SetController (Fortnite|GetFortPlayerControllerfromActor Actor)))
  (Utilities|IsValid _output_get
    (:Is%20Valid
      (if (GameplayTags|HasMatchingGameplayTag (Variables|GetActor) "(TagName=\"MovementModeExt.ThirdPerson\")")
        (return (Transformation|GetActorLocation (Variables|GetActor)) (Math|Conversions|RotationFromXVector (Transformation|GetActorForwardVector (Variables|GetActor))))
        (else
          (return (Transformation|GetActorLocation (Variables|GetActor)) (Pawn|GetDesiredRotation (LocalVariables|GetController))))))
    (:Is%20Not%20Valid
      (bind _output_get_2 (LocalVariables|SetAIController (AI|GetAIController Actor)))
      (Utilities|IsValid _output_get_2
        (:Is%20Valid
          (bind _returnvalue (TargetingSystem|GetPlayerCameraTargetingTransform (Variables|GetActor) true (1e+02 0.0 0.0)))
          (bind (_returnvalue_location _returnvalue_rotation) (Math|Transform|BreakTransform _returnvalue))
          (return _returnvalue_location _returnvalue_rotation))
        (:Is%20Not%20Valid
          (bind _returnvalue_2 (TargetingSystem|GetPlayerCameraTargetingTransform (Variables|GetActor) true (0.0 0.0 0.0)))
          (bind (_returnvalue_location_2 _returnvalue_rotation_2) (Math|Transform|BreakTransform _returnvalue_2))
          (return _returnvalue_location_2 _returnvalue_rotation_2))))))

(fn OverrideFailedReason
  (params (FailedReason :type Gameplay%20Tag%20Container%20Structure%20%28by%20ref%29))
  (bind _overridenfailedreason (|CallParentFunction FailedReason))
  (bind _abilityoncooldownsound (Variables|Default|GetAbilityOnCooldownSound)
    (:then
      (Audio|PlaySound2D _abilityoncooldownsound))
    (:else
      _))
  (return _overridenfailedreason))

(fn OnLaunchStarted
  (|CallParentFunction)
  (Utilities|IsValid (AI|GetAIController (CallFunction|GetAvatarasFortPlayerPawn))
    (:Is%20Valid
      (CallFunction|ApplyGuidedMotionMME)
      (GameplayEventRouter|BroadcastEvent (GameplayEventRouter|GetGlobalEventRouter) (Utilities|Struct|MakeIceGauntletsSpeechEvents "NewEnumerator1") 0))))
